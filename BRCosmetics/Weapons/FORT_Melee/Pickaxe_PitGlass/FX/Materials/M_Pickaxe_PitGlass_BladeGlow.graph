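# Translucent unlit blade glow material for the PitGlass pickaxe, built for Niagara mesh particles.
# Particle color feeds emissive, while particle alpha is multiplied by named mask terms to shape the glow.
# The masks are an instance-space directional gradient, a normal-map-based view-angle term, and a sharpened sine pulse driven by a dynamic parameter.
# A small camera-directed offset is computed to nudge the mesh toward the viewer.
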Material /BRCosmetics/Weapons/FORT_Melee/Pickaxe_PitGlass/FX/Materials/M_Pickaxe_PitGlass_BladeGlow
graph {
 properties {
  bUsedWithNiagaraMeshParticles=true
  blendMode=BLEND_Translucent
  pixelDepthOffsetMode=PDOM_Legacy
  shadingModel=MSM_Unlit
  usedShadingModels=Unlit
 }
 nodes [
  n1 = ParticleColor()
  MaterialFunctionCall(Emissive=@n1.RGB) {materialFunction=&MF_FX_EmissiveAdjustments}
  n2 = NamedRerouteUsage()
  n3 = NamedRerouteUsage()
  Multiply(A=@n1.A, B=Multiply(A=Multiply(A=@n2, B=@n3), B=NamedRerouteUsage()))
  n4 = CameraPositionWS()
  n5 = WorldPosition()
  Multiply(A=Normalize(VectorInput=Subtract(A=@n4, B=@n5.XYZ)), B=ScalarParameter() {defaultValue=0.025, parameterName=Camera Offset})
  n6 = Subtract(
   A=TransformPosition(None=WorldPosition().XYZ) {transformSourceType=TRANSFORMPOSSOURCE_World, transformType=TRANSFORMPOSSOURCE_Instance}
   B=VectorParameter() {parameterName=Origin Offset, defaultValue=<rgba 0 -5 55 0>}.RGB
  )
  n7 = Normalize(VectorInput=VectorParameter() {parameterName=Gradient Direction, defaultValue=<rgba 0 1 -0.6 0>}.RGB)
  n8 = Saturate(None=Divide(A=DotProduct(A=@n6, B=@n7), B=ScalarParameter() {defaultValue=55, parameterName=Glow Range}))
  NamedRerouteDeclaration(None=Power(Base=OneMinus(None=@n8), Exp=ScalarParameter() {defaultValue=3, parameterName=Glow Exponent})) {name=Gradient, nodeColor=<rgba 1 0 0.7031 1>}
  n9 = Transform(None=TextureSample() {samplerType=SAMPLERTYPE_Normal, texture=/BRCosmetics/Weapons/FORT_Melee/Pickaxe_PitGlass/Textures/T_PitGlass_Pickaxe_N}.RGB)
  n10 = CameraVectorWS()
  NamedRerouteDeclaration(None=Power(Base=DotProduct(A=@n9, B=@n10), Exp=ScalarParameter() {defaultValue=2, parameterName=Fresnel Exponent})) {name=Fresnel, nodeColor=<rgba 1 0.6094 0 1>}
  n11 = Divide(A=Saturate(None=DynamicParameter().Param1))
  NamedRerouteDeclaration(None=Power(Base=Sine(None=@n11)) {constExponent=9}) {name=Brightness, nodeColor=<rgba 1 0.5859 0 1>}
 ]
}
