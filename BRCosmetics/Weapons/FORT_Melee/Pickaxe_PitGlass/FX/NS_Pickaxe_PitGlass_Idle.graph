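# Niagara idle effect for the PitGlass pickaxe: soft cyan glows with flickering mesh lightning.
# SoftGlows bursts six persistent sprites, using partition-indexed arrays for offset, size and lights on two of them.
# SplineThickenMeshes randomly spawns short-lived bolts, picks head or non-head meshes and mirrors them via a User array.
# Lifetime curves drive material parameters, re-packed into per-particle colors and vectors for layered lightning glows across six mesh renderers.
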NiagaraSystem /BRCosmetics/Weapons/FORT_Melee/Pickaxe_PitGlass/FX/NS_Pickaxe_PitGlass_Idle
dependencies {
 usedDataInterfaces [&NiagaraDataInterfaceArrayFloat, &NiagaraDataInterfaceCurve, &NiagaraDataInterfaceArrayFloat3]
 usedDynamicInputs []
 usedModules [&SystemState]
}
emitters {
 SoftGlows {
  emitterUpdateScript [
   &EmitterState(
    Inactive Response=::NewEnumerator1
    Life Cycle Mode=::NewEnumerator1
    Loop Behavior=::NewEnumerator1
    Loop Duration Mode=::NewEnumerator1
    UseLoopDelay=false
   )
   &SpawnBurst_Instantaneous(Spawn Count=6, Use Spawn Probability=false, UseLoopCountLimit=true)
  ]
  particleSpawnScript [
   &PartitionParticles(Number Of Partitions=6)
   &InitializeParticle(
    Color=<rgba 0 1.1667 2 0.013>
    Position Offset=&SelectVectorFromArray(
     Direct Array Index=$Particles.PartitionParticles.Partition
     Vector Selection Array=DataInterface() {propertyValues {FloatData [<0 45 35>, <0 33 48>, <0 16 54>, <0 -8 60>, <0 6 38>, <0 8 -64>]}}
    )
    Sprite Size Mode=::NewEnumerator3
    Uniform Sprite Size=&SelectFloatFromArray(
     Direct Array Index=$Particles.PartitionParticles.Partition
     Float Selection Array=DataInterface() {propertyValues {FloatData [55,66,77,66,44,44]}}
    )
    UsePositionOffset=true
    Write Material Random=true
   )
   /SetVariables(
    Particles.LightColor=<rgba 0 1.3 2 1>
    Particles.LightRadius=&SelectFloatFromArray(
     Direct Array Index=$Particles.PartitionParticles.Partition
     Float Selection Array=DataInterface() {propertyValues {FloatData [0,0,75,0,0,33]}}
    )
    Particles.LightVisibilityTag:NiagaraInt32=&SelectIntFromArray(Direct Array Index=$Particles.PartitionParticles.Partition, Int Selection Array=DataInterface() {propertyValues {IntData [0,0,1,0,0,1]}})
   ) {isSetParameters=true}
  ]
  particleUpdateScript [
   &ParticleState(Kill Particles When Lifetime Has Elapsed=false, Let Infinitely Lived Particles Die When Emitter Deactivates=true)
   &CameraOffset(Camera Offset Amount=5)
  ]
  rendererClasses [&NiagaraSpriteRendererProperties, &NiagaraLightRendererProperties]
  renderers [
   {rendererClass=&NiagaraSpriteRendererProperties}
   {rendererClass=&NiagaraLightRendererProperties, rendererIndex=1}
  ]
 }
 SplineThickenMeshes {
  emitterUpdateScript [
   &EmitterState(
    Inactive Response=::NewEnumerator1
    Life Cycle Mode=::NewEnumerator1
    Loop Duration=&RandomRangeFloat(Minimum=0.05, Maximum=1.25)
    Recalculate Duration Each Loop=true
    UseLoopDelay=false
   )
   &SpawnRate(Spawn Probability=0.333, SpawnRate=3, Use Spawn Probability=true)
   &SpawnBurst_Instantaneous(Spawn Probability=0.5, Use Spawn Probability=true, UseLoopCountLimit=false)
  ]
  particleSpawnScript [
   /SetVariables(
    Particles.IsHead:NiagaraBool=&SetBoolByFloatComparison(A=&UniformAOrBFloat(A=-1, Distribution Weight=0.666), B=0)
    Particles.IsMirrored:NiagaraBool=&SetBoolByIntComparison(
     A=&SelectIntFromArray(
      Direct Array Index=&ReturnParticleIDAttributes()
      Direct Array Mode=::NewEnumerator1
      Int Selection Array=$User.LightningMirror
     )
     B=1
     Comparison Type=::NewEnumerator4
    )
   ) {isSetParameters=true}
   /SetVariables() {enabled=false, isSetParameters=true}
   /SetVariables(
    Particles.BladeGlowColor:LinearColor=<rgba 0 2.55 3 1>
    Particles.BladeGlowMeshScale:Vector3f=<1 1 1>
    Particles.BladeGlowVisibilityTag:NiagaraInt32=&MakeCustomIntFromBool(Bool=$Particles.IsHead, True INT=0, False INT=1)
    Particles.MeshIndex=&MakeCustomIntFromBool(
     Bool=$Particles.IsHead
     True INT=&ModuloInt(A=&ReturnParticleIDAttributes(), B=3)
     False INT=&Add_Integer(A=3, B=&ModuloInt(A=&ReturnParticleIDAttributes(), B=2))
    )
   ) {isSetParameters=true}
   /SetVariables(Particles.MeshIndex=1) {enabled=false, isSetParameters=true}
   &InitializeParticle(
    Color=<rgba 0 1 0.9942 1>
    Lifetime=&Multiply_Float(
     A=&RandomRangeFloat(Minimum=0.5, Maximum=0.5)
     B=&SelectFloatFromArray(Direct Array Index=$Particles.MeshIndex, Float Selection Array=DataInterface() {propertyValues {FloatData [0.5,1,0.666,0.666,0.8]}})
    )
    Mesh Scale=&MakeCustomVectorFromBool(Bool=$Particles.IsMirrored, True Vector=<-1 1 1>, False Vector=<1 1 1>)
    Mesh Scale Mode=::NewEnumerator4
    Position Mode=::NewEnumerator0
    Write Material Random=true
   )
  ]
  particleUpdateScript [
   &ParticleState()
   &DynamicMaterialParameters(
    Index 0 Param 1=&FloatFromCurve(FloatCurve=DataInterface() {propertyValues {Curve {keys [<at 0 0.25>, {arriveTangent=0.0283, interpMode=RCIM_Cubic, tangentMode=RCTM_Break, time=1, value=1}]}}})
    Index 0 Param 2=&FloatFromCurve(FloatCurve=DataInterface() {propertyValues {Curve {keys [<at 0 0.75>, {arriveTangent=-4.0866, time=1, value=0.25}]}}})
    Index 0 Param 3=&FloatFromCurve(FloatCurve=DataInterface() {propertyValues {Curve {keys [<at 0 1>,<at 1 0>]}}})
    Index 0 Param 4=&FloatFromCurve(FloatCurve=DataInterface() {propertyValues {Curve {keys [<at 0 1 RCIM_Constant>]}}}, Scale Curve=9)
    Index 1 Param 1=&FloatFromCurve(FloatCurve=DataInterface() {propertyValues {Curve {keys [<at 0 0>, {tangentMode=RCTM_User, time=1, value=2}]}}})
    Index 1 Param 2=&FloatFromCurve(
     FloatCurve=DataInterface() {propertyValues {Curve {keys [
         {interpMode=RCIM_Cubic, leaveTangent=25.2051, tangentMode=RCTM_Break, time=0, value=0}
         <at 0.2 2 RCIM_Constant>
         <at 0.266 1>
         <at 0.85 1>
         <at 1 0>
        ]}}}
     Scale Curve=&MakeCustomFloatFromBool(Bool=$Particles.IsHead, True Float=0.5, False Float=0.5)
    )
    Index 1 Param 3=&FloatFromCurve(FloatCurve=DataInterface() {propertyValues {Curve {keys [<at 0 0.25>,<at 0.5 0.25>,<at 1 2>]}}})
    Module.Index0Param0Write=true
    Module.Index0Param1Write=true
    Module.Index0Param2Write=true
    Module.Index0Param3Write=true
    Module.Index1Param0Write=true
    Module.Index1Param1Write=true
    Module.Index1Param2Write=true
    Module.Index1Param3Write=false
    Write Parameter Index 1=true
   )
   /SetVariables(
    Particles.LightningGlowColor:LinearColor=&MultiplyLinearColor(
     Linear Color=$Particles.Color
     Scale Factor=&MakeFloat4FromVectorAndFloat("Vector (RGB)"=&VectorFromFloat(Value=0.5), "Float (Alpha)"=1)
    )
    Particles.LightningGlowDynamic0:Vector4f=&MakeVector4(
     X=&MakeFloatFromVector4(Vector 4=$Particles.DynamicMaterialParameter, Channel=::NewEnumerator0)
     Y=&MakeFloatFromVector4(Vector 4=$Particles.DynamicMaterialParameter, Channel=::NewEnumerator1)
     Z=0
     W=0
    )
    Particles.LightningGlowDynamic1:Vector4f=&MakeVector4(
     X=&MakeFloatFromVector4(Vector 4=$Particles.DynamicMaterialParameter1, Channel=::NewEnumerator0)
     Y=&Multiply_Float(
      A=&MakeFloatFromVector4(Vector 4=$Particles.DynamicMaterialParameter1, Channel=::NewEnumerator1)
      B=1.666
     )
     Z=0.25
     W=0
    )
    Particles.LightningGlowSoftColor:LinearColor=&MultiplyLinearColor(
     Linear Color=$Particles.Color
     Scale Factor=&MakeFloat4FromVectorAndFloat("Vector (RGB)"=&VectorFromFloat(Value=0.166), "Float (Alpha)"=1)
    )
    Particles.LightningGlowSoftDynamic1:Vector4f=&MakeVector4(
     X=&MakeFloatFromVector4(Vector 4=$Particles.DynamicMaterialParameter, Channel=::NewEnumerator0)
     Y=&Multiply_Float(A=&MakeFloatFromVector4(Vector 4=$Particles.DynamicMaterialParameter1, Channel=::NewEnumerator1), B=3)
     Z=0
     W=0
    )
    Particles.LightningGlowSoftTopColor:LinearColor=<rgba 0.25 0.5 0.4866 0.2>
    Particles.LightningGlowSoftTopDynamic1:Vector4f=&MakeVector4(
     X=&MakeFloatFromVector4(Vector 4=$Particles.DynamicMaterialParameter, Channel=::NewEnumerator0)
     Y=&Multiply_Float(A=&MakeFloatFromVector4(Vector 4=$Particles.DynamicMaterialParameter1, Channel=::NewEnumerator1), B=3)
     Z=0
     W=0
    )
   ) {isSetParameters=true}
  ]
  rendererClasses [&NiagaraMeshRendererProperties]
  renderers [
   {rendererClass=&NiagaraMeshRendererProperties, rendererIndex=0}
   {rendererClass=&NiagaraMeshRendererProperties, rendererIndex=1}
   {rendererClass=&NiagaraMeshRendererProperties, rendererIndex=2}
   {rendererClass=&NiagaraMeshRendererProperties, rendererIndex=3}
   {rendererClass=&NiagaraMeshRendererProperties, rendererIndex=4}
   {rendererClass=&NiagaraMeshRendererProperties, rendererIndex=5}
  ]
 }
}
summary {
 systemName=NS_Pickaxe_PitGlass_Idle
 emitters [
  {emitterName=SplineThickenMeshes, rendererClasses [&NiagaraMeshRendererProperties]}
  {emitterName=SoftGlows, rendererClasses [&NiagaraSpriteRendererProperties, &NiagaraLightRendererProperties]}
 ]
 userVariables [
  User.bIsFrontEndPreview {type=NiagaraBool}
  User.bisInVehicle {type=NiagaraBool}
  User.IsWearingMask {type=NiagaraBool}
  User.LightningMirror {
   type=NiagaraDataInterfaceArrayInt32
   description=""
   defaultValue {
    struct=NiagaraExt_VariableValue_DataInterface
    value {
     dataInterface=&self:NiagaraDataInterfaceArrayInt32_1
     dataInterfaceClass=&NiagaraDataInterfaceArrayInt32
    }
   }
  }
  User.SkeletalMesh {
   type=NiagaraDataInterfaceSkeletalMesh
   defaultValue {
    struct=NiagaraExt_VariableValue_DataInterface
    value {dataInterface=&self:NiagaraDataInterfaceSkeletalMesh_0}
   }
  }
  User.Variant {type=NiagaraFloat, defaultValue=0, description=""}
 ]
}
